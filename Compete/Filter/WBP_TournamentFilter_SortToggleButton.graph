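# Toggle button for tournament filter and sort options, derived from the UIKit ButtonWithToggle.
# PreConstruct adjusts grid slots and padding, then disables text scrolling and sets a wrap width.
# Icons are set directly or async-loaded from soft textures, then repositioned in the grid; a material-parameter icon chain has no incoming execution.
# Selected, deselected and clicked events call the parent, then broadcast dispatchers passing the button itself.
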WidgetBlueprint /Compete/Filter/WBP_TournamentFilter_SortToggleButton
graphs [
 EventGraph {nodes [
   SetFilterIcon!(execute, self, Icon:Object Reference=Get ToggleIconMaterial() -> (ToggleIconMaterial:Material Instance Object Reference)) -> (then)
   > n1 = GetDynamicMaterial(self=Get Image!(self:WBP UIKit Block Image Object Reference=Get Block_Icon() -> (Block_Icon:WBP UIKit Block Image Object Reference)) -> (Image:Fort Mobile Image Object Reference))
   > SetTextureParameterValue(
    self=@n1
    ParameterName=Icon
    Value=Resolve Soft Reference!(Input:Texture Soft Object Reference=Get Texture() -> (Texture:Texture Soft Object Reference)) -> (Output:Texture Object Reference)
   )
   n2 = LoadIcon!() -> (OutputDelegate, then, Asset:Soft Object Reference) {is=title K2Node_CustomEvent}
   > n3 = Async Load Asset(Asset=@n2.Asset)
   SetIcon!(execute=@n3.Completed, self:WBP UIKit Button Generic Base Object Reference, Icon:Object Reference=@n3) -> (then)
   > SetIconSlotValues()
   n4 = Event PreConstruct()
   > Parent: PreConstruct(IsDesignTime=@n4.IsDesignTime)
   > SetSlotValues()
   n5 = Get TextBlock!(self:WBP UIKit Block Text Object Reference=Get Block_Text() -> (Block_Text:WBP UIKit Block Text Object Reference)) -> (TextBlock:Common Text Object Reference)
   > SetScrollingEnabled!(execute, self:Common Text Object Reference=@n5, bInIsScrollingEnabled:bool=false) -> (then)
   > SetWrapTextWidth!(execute, self:Common Text Object Reference=@n5, InWrapTextAt:int=360) -> (then)
   Event On Selected()
   > Parent: On Selected() {is=title K2Node_CallParentFunction}
   > Call On Toggle Selected!(
    execute:Exec
    self:WBP Tournament Filter Sort Toggle Button Object Reference
    Button:WBP Tournament Filter Sort Toggle Button Object Reference=Self-Reference()
   ) -> (then:Exec) {is=title K2Node_CallDelegate}
   Event On Clicked()
   > Parent: On Clicked() {is=title K2Node_CallParentFunction}
   > Call On Toggle Clicked!(
    execute:Exec
    self:WBP Tournament Filter Sort Toggle Button Object Reference
    Button:WBP Tournament Filter Sort Toggle Button Object Reference=Self-Reference()
   ) -> (then:Exec) {is=title K2Node_CallDelegate}
   Event On Deselected()
   > Parent: On Deselected() {is=title K2Node_CallParentFunction}
   > Call On Toggle Deselected!(
    execute:Exec
    self:WBP Tournament Filter Sort Toggle Button Object Reference
    Button:WBP Tournament Filter Sort Toggle Button Object Reference=Self-Reference()
   ) -> (then:Exec) {is=title K2Node_CallDelegate}
   Event On Pressed() {is=title K2Node_Event}
  ]}
 OnToggleClicked {nodes [OnToggleClicked!() -> (then, Button:WBP Tournament Filter Sort Toggle Button Object Reference) {is=title K2Node_FunctionEntry}]}
 OnToggleDeselected {nodes [OnToggleDeselected!() -> (then, Button:WBP Tournament Filter Sort Toggle Button Object Reference) {is=title K2Node_FunctionEntry}]}
 OnToggleSelected {nodes [OnToggleSelected!() -> (then, Button:WBP Tournament Filter Sort Toggle Button Object Reference) {is=title K2Node_FunctionEntry}]}
 SetFilterIcon {nodes [
   n1 = SetFilterIcon!() -> (then, Icon:Object Reference) {is=title K2Node_FunctionEntry}
   > SetIcon(Icon=@n1)
   > SetIconSlotValues()
  ]}
 SetFilterIconFromTexture {nodes [
   n1 = SetFilterIconFromTexture!() -> (then, Texture:Texture Soft Object Reference) {is=title K2Node_FunctionEntry}
   > LoadIcon!(execute, self, Asset:Soft Object Reference=@n1) -> (then) {is=title K2Node_CallFunction}
  ]}
 SetIconSlotValues {nodes [
   SetIconSlotValues() {is=title K2Node_FunctionEntry}
   > n1 = Get!(execute:Exec) -> (then:Exec, else:Exec, Block_Icon:WBP UIKit Block Image Object Reference)
   > SetGridSlotColumnAndRow!(execute, Widget:Widget Object Reference=@n1, Row:int=1, Row Span:int=0, Column:int=0, Column Span:int=0) -> (then)
  ]}
 SetSlotValues {nodes [
   SetSlotValues() {is=title K2Node_FunctionEntry}
   > SetGridSlotColumnAndRow(Widget=Get Slot_Main() -> (Slot_Main:Named Slot Object Reference), Row=1, Row Span=0, Column=0, Column Span=0)
   n1 = SlotAsGridSlot(Widget=Get Main() -> (Main:Grid Panel Object Reference))
   n2 = Get!(self:Grid Slot Object Reference=@n1) -> (Padding_Left:float=~, Padding_Top:float=~, Padding_Right:float=~, Padding_Bottom:float=~)
   > Set!(
    execute
    Padding_Left:float=Get MainPaddingLeft() -> (MainPaddingLeft:float)
    Padding_Top:float=@n2.Padding_Top
    Padding_Right:float=0
    Padding_Bottom:float=@n2.Padding_Bottom
    self:Grid Slot Object Reference=@n1
   ) -> (then, Output_Get:Margin Structure)
  ]}
]
variables [
 FilterName {type=string, default=""}
 MainPaddingLeft {type=number, default=16}
 OnToggleClicked {}
 OnToggleDeselected {}
 OnToggleSelected {}
 ToggleIconMaterial {type=MaterialInstance, default=/Compete/Filter/Materials/MI_UI_Tournament_TypeIcon}
]
widgets {
 info {
  inheritedWidgetCount=0
  parentClass=/FortUILibrary/UIKit/Buttons/WBP_UIKit_ButtonWithToggle.WBP_UIKit_ButtonWithToggle_C
  rootWidgetClass=null
  widgetCount=0
 }
 widgets []
}
