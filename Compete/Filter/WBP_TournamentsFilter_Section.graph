WidgetBlueprint /Compete/Filter/WBP_TournamentsFilter_Section
graphs [
 BindToggleDelegates {nodes [
   BindToggleDelegates!() -> (then, Button:WBP Tournament Filter Sort Toggle Button Object Reference) {is=title K2Node_FunctionEntry}
   > n1 = Branch(Condition=Get bRadioType() -> (bRadioType:bool))
   > Bind Event to On Toggle Clicked(
    self:WBP Tournament Filter Sort Toggle Button Object Reference=Get Button() -> (Button:WBP Tournament Filter Sort Toggle Button Object Reference)
    Delegate=Create Event()
   ) {is=title K2Node_AddDelegate}
   Bind Event to On Toggle Selected(execute=@n1.else, self:WBP Tournament Filter Sort Toggle Button Object Reference=Get Button(), Delegate=Create Event()) {is=title K2Node_AddDelegate}
   > Bind Event to On Toggle Deselected(self:WBP Tournament Filter Sort Toggle Button Object Reference=Get Button(), Delegate=Create Event()) {is=title K2Node_AddDelegate}
  ]}
 Create Filters Buttons {nodes [
   n1 = Create Filters Buttons!() -> (then, FilterData:Tournaments Filter Section Structure) {is=title K2Node_FunctionEntry}
   > ClearChildren(self=Get UniformGridPanel_Filters() -> (UniformGridPanel_Filters:Uniform Grid Panel Object Reference))
   n2 = Break Tournaments Filter Section!(TournamentsFilterSection:"Tournaments Filter Section Structure (by ref)"=@n1) -> (SectionName:Text, FilterOptions:Array of Tournament Type With Name Structures, bRadioType:bool=~) {is=title K2Node_BreakStruct}
   > SetFilterSectionName!(execute, self, InText:Text=@n2) -> (then)
   n3 = Get ToggleValues() -> (ToggleValues:Array of Booleans)
   > Clear(TargetArray:Array of Booleans=@n3)
   > Resize!(execute, TargetArray:Array of Booleans=@n3, Size:int=Length(TargetArray:Array of Tournament Type With Name Structures=@n2.FilterOptions)) -> (then) {is=title K2Node_CallArrayFunction}
   > n4 = For Each Loop(Array:Array of Tournament Type With Name Structures=@n2.FilterOptions) -> (Array Element:Tournament Type With Name Structure)
   > Set Index(Index:int=@n4.Array Index) -> (Output_Get:int=0)
   > n5 = Create Widget!(
    execute:Exec
    Class:User Widget Class Reference=&/WBP_TournamentFilter_SortToggleButton.WBP_TournamentFilter_SortToggleButton_C
    OwningPlayer:Player Controller Object Reference
    Style:Common Button Style Class Reference=/FortUILibrary/UIKit/Buttons/Blocks/ButtonStyle_UIKit_Empty.ButtonStyle_UIKit_Empty_C
    bApplyAlphaOnDisable:bool=False
    bLocked:bool=False
    bSelectable:bool=True
    bShouldSelectUponReceivingFocus:bool=False
    bInteractableWhenSelected:bool=True
    bToggleable:bool=true
    bTriggerClickedAfterSelection:bool=False
    bDisplayInputActionWhenNotInteractable:bool=False
    bHideInputActionWithKeyboard:bool=False
    bShouldUseFallbackDefaultInputAction:bool=False
    bRequiresHold:bool=False
    HoldData:Common UIHold Data Class Reference=&CommonUI_HoldData_Base_C
    ClickMethod:EButtonClickMethod Enum=MouseDown
    InputPriority:int=0
    TriggeringInputAction:Data Table Row Handle Structure
   ) -> (then:Exec, ReturnValue:WBP Tournament Filter Sort Toggle Button Object Reference)
   > n6 = Set ToggleButton(ToggleButton:WBP Tournament Filter Sort Toggle Button Object Reference=@n5) -> (Output_Get:WBP Tournament Filter Sort Toggle Button Object Reference)
   n7 = Break Tournament Type with Name!(TournamentTypeWithName:"Tournament Type With Name Structure (by ref)"=@n4) -> (Name:Text, Type:ETournamentTypeFilter Enum) {is=title K2Node_BreakStruct}
   > Set Text(Text=@n7, self:WBP UIKit Button Generic Base Object Reference=@n6)
   > Add(
    TargetArray:Array of WBP Tournament Filter Sort Toggle Button Object References=Get FilterSectionEntries() -> (FilterSectionEntries:Array of WBP Tournament Filter Sort Toggle Button Object References)
    NewItem:"WBP Tournament Filter Sort Toggle Button Object Reference (by ref)"=Get ToggleButton() -> (ToggleButton:WBP Tournament Filter Sort Toggle Button Object Reference)
   ) {is=title K2Node_CallArrayFunction}
   > SetTogglePosition!(execute, self, ToggleButton:Widget Object Reference=Get ToggleButton(), Index:int=Get Index() -> (Index:int=0)) -> (then)
   > n8 = Set bRadioType(bRadioType:bool=Break Tournaments Filter Section(TournamentsFilterSection=Get FilterData() -> (FilterData:Tournaments Filter Section Structure)) -> (bRadioType=~).bRadioType) -> (Output_Get:bool)
   > SetToggleType!(
    execute
    self:WBP UIKit Button With Toggle Object Reference=Get ToggleButton()
    Type:E_UI_ToggleType Enum=Select(Option 0:E_UI_ToggleType Enum=NewEnumerator1, Option 1:E_UI_ToggleType Enum=NewEnumerator0, Index=@n8) -> (ReturnValue:E_UI_ToggleType Enum)
   ) -> (then)
   > n9 = Branch(Condition=Get IsShowIcons() -> (IsShowIcons:bool))
   > n10 = GetFilterIconFromEnum!(execute, self, FilterEnum:ETournamentTypeFilter Enum=@n7.Type) -> (then, Texture:Texture Soft Object Reference)
   > n11 = SetFilterIconFromTexture!(execute, self:WBP Tournament Filter Sort Toggle Button Object Reference=Get ToggleButton(), Texture:Texture Soft Object Reference=@n10) -> (then)
   BindToggleDelegates!(execute=[@n11.then, @n9.else], self, Button:WBP Tournament Filter Sort Toggle Button Object Reference=Get ToggleButton()) -> (then) {is=title K2Node_CallFunction}
   Set SelectedRadioToggle(SelectedRadioToggle:WBP UIKit Button With Toggle Object Reference) -> (Output_Get:WBP UIKit Button With Toggle Object Reference)
  ]}
 EventGraph {nodes [Event PreConstruct()]}
 Get Selected Filters {nodes [
   Get Selected Filters() {is=title K2Node_FunctionEntry}
   > n1 = For Each Loop(Array:Array of Widget Object References=GetAllChildren(self=Get UniformGridPanel_Filters())) -> (Array Element:Widget Object Reference)
   Return Node!(
    execute=@n1.Completed
    Filters:Array of WBP Tournament Filter Sort Toggle Button Object References=Get SelectedFilters() -> (SelectedFilters:Array of WBP Tournament Filter Sort Toggle Button Object References)
   )
   > n2 = Cast To WBP_TournamentFilter_SortToggleButton!(execute, Object=@n1) -> (then, CastFailed, AsWBP Tournament Filter Sort Toggle Button:WBP Tournament Filter Sort Toggle Button Object Reference) {is=title K2Node_DynamicCast}
   > Branch(Condition=GetSelected!(self:Common Button Base Object Reference=@n2) -> (ReturnValue:bool))
   > Add(TargetArray=Get SelectedFilters(), NewItem=@n2)
  ]}
 GetFilterIconFromEnum {nodes [
   n1 = GetFilterIconFromEnum!() -> (then, FilterEnum:ETournamentTypeFilter Enum=All) {is=title K2Node_FunctionEntry}
   > Return Node!(
    execute
    Texture:Texture Soft Object Reference=Find(
     TargetMap:Map of ETournamentTypeFilter Enums to Texture Soft Object References=Get TournamentTypesFilterIcons() -> (TournamentTypesFilterIcons:Map of ETournamentTypeFilter Enums to Texture Soft Object References=All)
     Key:"ETournamentTypeFilter Enum (by ref)"=@n1
    ) -> (Value:Texture Soft Object Reference)
   )
  ]}
 IsFilterEntriesEmpty {nodes [
   IsFilterEntriesEmpty() {is=title K2Node_FunctionEntry}
   > Return Node!(
    execute
    bEmpty:bool=Is Empty!(TargetArray:Array of WBP Tournament Filter Sort Toggle Button Object References=Get FilterSectionEntries()) -> (ReturnValue:bool)
   )
  ]}
 OnToggleClicked {nodes [
   n1 = OnToggleClicked!() -> (then, Button:WBP Tournament Filter Sort Toggle Button Object Reference) {is=title K2Node_FunctionEntry}
   n2 = Get Text(self:WBP UIKit Button Generic Base Object Reference=Get Button())
   > Fire Event Element Interacted(
    Element=@n1
    AdditionalInfo="Add String (Non-Localized)"!(
     Info:"Flow Analytics Additional Info Structure (by ref)"=Make Flow Analytics Additional Info!() -> (FlowAnalyticsAdditionalInfo:Flow Analytics Additional Info Structure) {is=title K2Node_MakeStruct}
     Key:String=ToggleName
     Value:String="To String (Text)"(InText=@n2)
    ) -> (ReturnValue:Flow Analytics Additional Info Structure)
   )
   > n3 = Sequence()
   > n4 = Get!(execute:Exec) -> (then:Exec, else:Exec, SelectedRadioToggle:WBP UIKit Button With Toggle Object Reference)
   > n5 = ClearSelection(self:Common Button Base Object Reference=@n4)
   Set SelectedRadioToggle(execute=[@n5.then, @n4.else], SelectedRadioToggle=Get Button())
   n6 = Find Item!(
    TargetArray:Array of WBP Tournament Filter Sort Toggle Button Object References=Get FilterSectionEntries()
    ItemToFind:"WBP Tournament Filter Sort Toggle Button Object Reference (by ref)"=@n1
   ) -> (ReturnValue)
   Branch(execute=@n3.then_1, Condition="integer >= integer"(A=@n6))
   n7 = Get ToggleValues()
   > n8 = For Each Loop(Array:Array of Booleans=@n7) -> (Array Element:bool)
   > Set Array Elem!(
    execute
    TargetArray:Array of Booleans=@n7
    Index:int=@n8.Array Index
    Item:"Boolean (by ref)"="Equal (Integer)"(A=@n8.Array Index, B=@n6)
    bSizeToFit:bool=false
   ) -> (then)
  ]}
 OnToggleDeselected {nodes [
   n1 = OnToggleDeselected!() -> (then, Button:WBP Tournament Filter Sort Toggle Button Object Reference) {is=title K2Node_FunctionEntry}
   n2 = Find Item(TargetArray=Get FilterSectionEntries(), ItemToFind=@n1)
   > Branch(Condition="integer >= integer"(A=@n2))
   > Set Array Elem(TargetArray=Get ToggleValues(), Index=@n2, bSizeToFit=false)
   > Branch(Condition=Is Empty(TargetArray=Get Selected Filters!() -> (Filters:Array of WBP Tournament Filter Sort Toggle Button Object References) {is=title K2Node_CallFunction}))
   > SetIsSelected!(
    execute
    self:Common Button Base Object Reference="Get (a copy)"(Array:Array of WBP Tournament Filter Sort Toggle Button Object References=Get FilterSectionEntries()) -> (Output:WBP Tournament Filter Sort Toggle Button Object Reference)
    InSelected:bool=true
    bGiveClickFeedback:bool=false
   ) -> (then)
  ]}
 OnToggleSelected {nodes [
   n1 = OnToggleSelected!() -> (then, Button:WBP Tournament Filter Sort Toggle Button Object Reference) {is=title K2Node_FunctionEntry}
   n2 = Get Text(self=Get Button())
   > Fire Event Element Interacted(
    Element=@n1
    AdditionalInfo="Add String (Non-Localized)"(Info=Make Flow Analytics Additional Info(), Key=ToggleName, Value="To String (Text)"(InText=@n2))
   )
   n3 = Find Item(TargetArray=Get FilterSectionEntries(), ItemToFind=@n1)
   > Branch(Condition="integer >= integer"(A=@n3))
   > Set Array Elem(TargetArray=Get ToggleValues(), Index=@n3, Item=true, bSizeToFit=false)
   > n4 = Branch(Condition="integer > integer"(A=@n3))
   n5 = Get ToggleValues()
   > Branch(Condition="Get (a ref)"(Array:Array of Booleans=@n5, Dimension 1=0) -> (Output:"Boolean (by ref)"))
   > Set Array Elem(TargetArray=Get ToggleValues(), Index=0, bSizeToFit=false)
   > SetIsSelected(self="Get (a copy)"(Array=Get FilterSectionEntries()), InSelected=false, bGiveClickFeedback=false)
   n6 = For Each Loop(Exec=@n4.else, Array=@n5)
   > Branch(Condition="integer > integer"(A=@n6.Array Index))
   > Set Array Elem(TargetArray=@n5, Index=@n6.Array Index, bSizeToFit=false)
   > SetIsSelected(self="Get (a copy)"(Array=Get FilterSectionEntries(), Dimension 1=@n6.Array Index), InSelected=false, bGiveClickFeedback=false)
  ]}
 SetFilterSectionName {nodes [
   n1 = SetFilterSectionName!() -> (then, InText:Text) {is=title K2Node_FunctionEntry}
   > "SetText (Text)"(self=Get Text_SectionName() -> (Text_SectionName:Common Text Object Reference), InText=@n1)
  ]}
 SetToggleEnabled {nodes [
   SetToggleEnabled!() -> (then, bEnabled:bool=~, Index:int) {is=title K2Node_FunctionEntry}
   n1 = "Get (a copy)"(Array=Get FilterSectionEntries(), Dimension 1=Get Index() -> (Index=0))
   > Is Valid(InputObject=@n1)
   > Set bIsEnabled!(execute, bIsEnabled:bool=Get bEnabled() -> (bEnabled:bool), self:Widget Object Reference=@n1) -> (then, Output_Get:bool)
  ]}
 SetTogglePosition {nodes [
   n1 = SetTogglePosition!() -> (then, ToggleButton:Widget Object Reference, Index:int) {is=title K2Node_FunctionEntry}
   > n2 = AddChildToUniformGrid!(
    execute
    self:Uniform Grid Panel Object Reference=Get UniformGridPanel_Filters()
    Content:Widget Object Reference=@n1
    InRow:int=int / int(A=Get Index() -> (Index=0), B=Get MaxElementsPerRow() -> (MaxElementsPerRow:int=0))
    InColumn:int="% (Integer)"(A=Get Index() -> (Index=0), B=Get MaxElementsPerRow() -> (MaxElementsPerRow=0)) -> (ReturnValue=0)
   ) -> (then, ReturnValue:Uniform Grid Slot Object Reference)
   > SetHorizontalAlignment(self:Uniform Grid Slot Object Reference=@n2, InHorizontalAlignment=HAlign_Left)
  ]}
 SetToggleSelectionStateByIndex {nodes [
   SetToggleSelectionStateByIndex!() -> (then, bSelected:bool=~, Index:int) {is=title K2Node_FunctionEntry}
   n1 = "Get (a copy)"(Array=Get FilterSectionEntries(), Dimension 1=Get Index() -> (Index=0))
   > Is Valid(InputObject=@n1)
   > SetIsSelected(self=@n1, InSelected=Get bSelected() -> (bSelected:bool), bGiveClickFeedback=false)
  ]}
 SetToggleType {nodes [
   n1 = SetToggleType!() -> (then, ToggleType:E_UI_ToggleType Enum=NewEnumerator0) {is=title K2Node_FunctionEntry}
   > Set bRadioType(bRadioType="Equal (Enum)"(A:E_UI_ToggleType Enum=@n1, B:E_UI_ToggleType Enum=NewEnumerator0))
   > n2 = For Each Loop(Array:Array of WBP Tournament Filter Sort Toggle Button Object References=Get FilterSectionEntries()) -> (Array Element:WBP Tournament Filter Sort Toggle Button Object Reference)
   > SetToggleType(self=@n2, Type=Get ToggleType() -> (ToggleType:E_UI_ToggleType Enum=NewEnumerator0)) {is=title K2Node_CallFunction}
  ]}
 UpdateToggleValues {nodes [
   n1 = UpdateToggleValues!() -> (then, Values:Array of Booleans) {is=title K2Node_FunctionEntry}
   n2 = Length(TargetArray:Array of Booleans=@n1)
   n3 = GetChildrenCount!(self:Panel Widget Object Reference=Get UniformGridPanel_Filters()) -> (ReturnValue:int=0)
   > Branch(Condition=AND Boolean!(A="integer > integer"(A=@n2), B="Equal (Integer)"(A=@n2, B=@n3), C:bool="integer > integer"(A=@n3)) -> (ReturnValue))
   n4 = Get ToggleValues()
   > Clear(TargetArray=@n4)
   > Resize(TargetArray=@n4, Size=GetChildrenCount(self=Get UniformGridPanel_Filters()) -> (ReturnValue=0))
   > n5 = For Each Loop(Array:Array of Widget Object References=GetAllChildren(self=Get UniformGridPanel_Filters())) -> (Array Element:Widget Object Reference)
   > n6 = Cast To WBP_TournamentFilter_SortToggleButton(Object=@n5)
   n7 = "Get (a copy)"(Array:Array of Booleans=Get Values() -> (Values:Array of Booleans), Dimension 1=@n5.Array Index) -> (Output:bool)
   > SetIsSelected(self=@n6, InSelected=@n7, bGiveClickFeedback=false)
   > Set Array Elem(TargetArray=Get ToggleValues(), Index=@n5.Array Index, Item=@n7, bSizeToFit=false)
   > Branch(Condition=AND Boolean!(A=@n7, B=Get bRadioType()) -> (ReturnValue))
   > Set SelectedRadioToggle(SelectedRadioToggle=@n6)
  ]}
]
variables [
 bRadioType {type=boolean}
 FilterSectionEntries {type="array<WBP_TournamentFilter_SortToggleButton_C>", default []}
 IsShowIcons {type=boolean}
 MaxElementsPerRow {type=integer, default=3}
 SelectedRadioToggle {type=WBP_UIKit_ButtonWithToggle_C, default=null}
 ToggleValues {type="array<boolean>", default []}
 TournamentTypesFilterIcons {default {
   all=null
   cashCup=/FortUILibrary/Textures/Icons/T_UI_Icon_UIKit_CashCup_DF
   fNCS=/FortUILibrary/Textures/Icons/T_UI_Icon_UIKit_FNCS_DF
   otherOrNone=/FortUILibrary/Textures/Icons/T_UI_Icon_UIKit_OtherTournaments_DF
   rankedCup=/FortUILibrary/Textures/Icons/T_UI_Icon_UIKit_RankedCup_DF
   shopCup=/FortUILibrary/Textures/Icons/T_UI_Icon_UIKit_ItemShopCup_DF
   victoryCup=/FortUILibrary/Textures/Icons/T_UI_Icon_UIKit_VictoryCup_DF
  }}
 TournamentTypesIcons {default {
   cashCup=&T_UI_Icon_UIKit_CashCup_DF
   fNCS=&T_UI_Icon_UIKit_FNCS_DF
   none=null
   other=&T_UI_Icon_UIKit_OtherTournaments_DF
   rankedCup=&T_UI_Icon_UIKit_RankedCup_DF
   shopCup=/FortUILibrary/Textures/Icons/T_UI_Icon_UIKit_Shop_SDF
   victoryCup=&T_UI_Icon_UIKit_VictoryCup_DF
  }}
]
widgets {
 info {inheritedWidgetCount=0, parentClass=&CommonUserWidget, rootWidgetClass=&VerticalBox, widgetCount=3}
 widgets [
  Root {widgetClassPath=&VerticalBox, widget=Root}
  Text_SectionName {
   widgetClassPath=&CommonTextBlock
   bIsVariable=true
   parent=Root
   widget=Text_SectionName
   properties {
    bAutoCollapseWithEmptyText=true
    clipping=OnDemand
    scrollStyle=&TextScrollStyle-Base_C
    text="[PH] Section Name"
    textOverflowPolicy=Ellipsis
    font {
     bForceMonospaced=false
     bMaterialIsStencil=false
     fontMaterial=null
     fontObject=&HeadingNow
     letterSpacing=0
     monospacedWidth=1
     size=20.25
     skewAmount=0
     typefaceFontName=75 Medium
     outlineSettings {
      bApplyOutlineToDropShadows=false
      bMiteredCorners=false
      bSeparateFillAlpha=false
      outlineMaterial=null
      outlineSize=0
      outlineColor=<rgba 0 0 0 1>
     }
    }
    shadowOffset=<0 0>
   }
   slot {class=VerticalBoxSlot, properties {padding=<ltrb 18 0 0 0>}}
  }
  UniformGridPanel_Filters {
   widgetClassPath=/Script/UMG.UniformGridPanel
   bIsVariable=true
   parent=Root
   widget=UniformGridPanel_Filters
   properties {slotPadding=<ltrb 18 0 5 24>}
   slot {class=VerticalBoxSlot, properties {padding=<ltrb 0 25 0 7>}}
  }
 ]
}
